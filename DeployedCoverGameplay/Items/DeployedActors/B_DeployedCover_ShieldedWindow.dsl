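; Deployable cover actor with an armored window that players open and close by interacting.
; Interaction toggles the replicated closed state and starts a cooldown, and is allowed only off cooldown, undestroyed and when the pawn faces roughly along the cover's forward.
; OnRep snaps the window's Z scale on authority, while clients animate it with a duration-scaled timeline.
; Server-detected zero health sets a destroyed flag whose OnRep hides the window and disables its collision.
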(event Custom|InitArmoredWindow
  (Networking|SwitchHasAuthority
    (:Remote
      (CallFunction|SetWindowScale (Math|Float|SelectFloat (Variables|Default|GetWindowClosedScale) (Variables|Default|GetWindowOpenScale) (Variables|Default|GetbWindowClosed))))))

(event EventBeginPlay
  (|CallParentFunction)
  (CallFunction|InitArmoredWindow)
  (Networking|SwitchHasAuthority
    (:Authority
      (EventDispatchers|BindEventtoOnDamaged (EventDispatchers|CreateEvent :!event Custom|OnDamageReceived)))))

(event Interaction|EventBlueprintOnInteract
  (params (InteractingPawn :type Fort%20Pawn%20Object%20Reference) (InteractionBeingAttempted :type EInteractionBeingAttempted%20Enum :default "FirstInteraction"))
  (Variables|Default|SetbWindowClosed (not (Variables|Default|GetbWindowClosed)))
  (Variables|Default|SetbWindowCanUseAfterCooldown false)
  (Utilities|Time|SetTimerbyEvent (delegate Custom|ResetWindowCooldown) (ScalableFloat|GetValueatLevel (Variables|Default|GetHF_WindowCooldownTime))))

(event Custom|ResetWindowCooldown
  (Variables|Default|SetbWindowCanUseAfterCooldown true))

(event Custom|HideImmediatelyOnDestroyed
  (bind _armoredwindow (Variables|Default|GetArmoredWindow))
  (Rendering|SetVisibility _armoredwindow false false)
  (Collision|SetCollisionEnabled _armoredwindow "NoCollision"))

(event Custom|OnDamageReceived
  (params (DamagedActor :type Actor%20Object%20Reference) (Damage :type Float%20%28single-precision%29) (InstigatedBy :type Controller%20Object%20Reference) (DamageCauser :type Actor%20Object%20Reference) (HitLocation :type Vector) (FHitComponent :type Primitive%20Component%20Object%20Reference) (BoneName :type Name) (Momentum :type Vector))
  (if (<= (Health|GetHealthPercent) 0)
    (Variables|Default|SetbHasBeenDestroyed true)))

(event Custom|AnimateWindowState_CLIENT
  (bind _bwindowclosed (Variables|Default|GetbWindowClosed))
  (bind _returnvalue (not _bwindowclosed))
  (Components|Timeline|SetPlayRate (Variables|B_DeployedCover_ShieldedWindow|GetAnimateWindow) (Math|Float|SafeDivide 1.0 (Math|Float|SelectFloat (ScalableFloat|GetValueatLevel (Variables|Default|GetHF_OpenAnimationTime)) (ScalableFloat|GetValueatLevel (Variables|Default|GetHF_CloseAnimationTime)) _returnvalue)))
  (if _returnvalue
    (bind (_direction _windowscale) (Timeline|AnimateWindow Play)
      (:Update
        (CallFunction|SetWindowScale (Math|Float|Lerp (Variables|Default|GetWindowClosedScale) (Variables|Default|GetWindowOpenScale) _windowscale))))
    (else
      (Timeline|AnimateWindow Reverse))))

(fn BlueprintCanInteract
  (params (InteractingPawn :type Fort%20Pawn%20Object%20Reference) (InteractionBeingAttempted :type EInteractionBeingAttempted%20Enum :default "FirstInteraction") (InteractionType :type TInteractionType%20Enum :default "IT_NoInteraction"))
  (return (Math|Boolean|ANDBoolean (not (Variables|Default|GetbHasBeenDestroyed)) (> (Math|Vector|DotProduct (Transformation|GetActorForwardVector InteractingPawn) (Transformation|GetActorForwardVector)) 0.2) (Variables|Default|GetbWindowCanUseAfterCooldown))))

(fn BlueprintGetInteractionString
  (params (InteractingPawn :type Fort%20Pawn%20Object%20Reference) (InteractionBeingAttempted :type EInteractionBeingAttempted%20Enum :default "FirstInteraction"))
  (return (Utilities|String|SelectText "NSLOCTEXT(\"[B506131FC5FBBC3B1676FE2C93947E4F]\", \"F589DA484EE2E659B22D37AE7919BE22\", \"Open\")" "NSLOCTEXT(\"[B506131FC5FBBC3B1676FE2C93947E4F]\", \"ECA2A5DC4C850E9E1E5F35AD9E7D4931\", \"Close\")" (Variables|Default|GetbWindowClosed))))

(fn BlueprintGetInteractionTime
  (params (InteractingPawn :type Fort%20Pawn%20Object%20Reference) (InteractionBeingAttempted :type EInteractionBeingAttempted%20Enum :default "FirstInteraction"))
  (return (Math|Float|SelectFloat (ScalableFloat|GetValueatLevel (Variables|Default|GetHF_OpenInteractTime)) (ScalableFloat|GetValueatLevel (Variables|Default|GetHF_CloseInteractTime)) (Variables|Default|GetbWindowClosed)) true))

(fn GetIconPlacement
  (params (SelfActor :type Actor%20Object%20Reference) (ViewingActor :type Actor%20Object%20Reference))
  ;[0.15 0.08906 0.1458 0.5] FORT-863654: Mobile-only fix for the Easy Interact button, which was obstructing the player's view when opening the window.
  (return (- (Physics|GetCenterOfMass (Variables|Default|GetInteractionArea) "None") "0.000000,0.000000,25.000000") "0, 0, 0"))

(fn OnRep_bHasBeenDestroyed
  (CallFunction|HideImmediatelyonDestroyed))

(fn OnRep_bWindowClosed
  (Networking|SwitchHasAuthority
    (:Authority
      (CallFunction|SetWindowScale (Math|Float|SelectFloat (Variables|Default|GetWindowClosedScale) (Variables|Default|GetWindowOpenScale) (Variables|Default|GetbWindowClosed))))
    (:Remote
      (CallFunction|AnimateWindowStateCLIENT))))

(fn SetWindowScale
  (params (OpenZScale :type Float%20%28double-precision%29))
  (Transformation|SetRelativeScale3D (Variables|Default|GetArmoredWindow) (Math|Vector|MakeVector 1.6 0.26 OpenZScale)))

(fn ConstructionScript)
