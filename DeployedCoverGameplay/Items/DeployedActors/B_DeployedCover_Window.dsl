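; Deployed cover actor with an armored window that opens and closes, interactable by non-downed pawns facing along its right vector when intact and off cooldown.
; The client animation event flips the open state, drives material-based world position offset, toggles collision and health bar, and plays curtain sounds.
; Where cosmetics cannot run, SetWindowScale sets collision from the closed flag and can nudge overlapping players upward into falling.
; Destruction there fires a replicated GameplayCue.
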(event EventBeginPlay
  (bind _returnvalue (Utilities|Time|GetGameTimeinSeconds))
  (bind _armoredwindow (Variables|Default|GetArmoredWindow))
  (|CallParentFunction)
  (Rendering|Material|SetScalarParameterValueonMaterials _armoredwindow "BP_GetTime" _returnvalue)
  (Rendering|Material|SetScalarParameterValueonMaterials _armoredwindow "BP_GetTime_Scale" _returnvalue))

(event Collision|EventActorBeginOverlap
  (params (OtherActor :type Actor%20Object%20Reference)))

(event EventTick
  (params (DeltaSeconds :type Float%20%28single-precision%29)))

(event EventAnimateWindowState_CLIENT
  (bind _armoredwindow (Variables|Default|GetArmoredWindow))
  (bind _returnvalue (Transformation|GetWorldLocation _armoredwindow))
  (bind _armoredwindow_2 (Variables|Default|GetArmoredWindow))
  (bind _output_get (Variables|Default|SetIsOpened (- 1.0 (Variables|Default|GetIsOpened))))
  (Rendering|Material|SetScalarParameterValueonMaterials (Variables|Default|GetArmoredWindow) "BP_WPO_Window_CLOSE/OPEN" _output_get)
  (Rendering|Material|SetScalarParameterValueonMaterials _armoredwindow_2 "AnimationDelay_Window" 0.0)
  (Rendering|Material|SetScalarParameterValueonMaterials _armoredwindow_2 "BP_GetTime" (Utilities|Time|GetGameTimeinSeconds))
  (if (== (Variables|Default|GetIsOpened) 1.0)
    (Collision|SetCollisionEnabled _armoredwindow "NoCollision")
    (CallFunction|SetSuppressHealthBar true)
    (Audio|PlaySoundatLocation "/DeployedCoverGameplay/Assets/Audio/SC_DeployedCover_Curtain_Off.SC_DeployedCover_Curtain_Off" _returnvalue)
    (else
      (Collision|SetCollisionEnabled _armoredwindow "QueryAndPhysics")
      (CallFunction|SetSuppressHealthBar false)
      (Audio|PlaySoundatLocation "/DeployedCoverGameplay/Assets/Audio/SC_DeployedCover_Curtain_On.SC_DeployedCover_Curtain_On" _returnvalue))))

(event EventHideImmediatelyOnDestroyed
  (|CallParentFunction)
  (Networking|CanExecuteCosmeticEvents
    (:False
      (Ability|ExecuteGameplayCueReplicated (Variables|Default|GetGCN_Destroyed) (Utilities|Struct|MakeGameplayEffectContextHandle)))))

(fn BlueprintCanInteract
  (params (InteractingPawn :type Fort%20Pawn%20Object%20Reference) (InteractionBeingAttempted :type EInteractionBeingAttempted%20Enum :default "FirstInteraction") (InteractionType :type TInteractionType%20Enum :default "IT_NoInteraction"))
  (return (Math|Boolean|ANDBoolean (not (Variables|Default|GetbHasBeenDestroyed)) (> (Math|Vector|DotProduct (Transformation|GetActorForwardVector InteractingPawn) (Transformation|GetActorRightVector)) 0.2) (Variables|Default|GetbWindowCanUseAfterCooldown) (not (Health|IsDBNO InteractingPawn)))))

(fn SetWindowScale
  (params (OpenZScale :type Float%20%28double-precision%29))
  (bind _armoredwindow (Variables|Default|GetArmoredWindow))
  (Networking|CanExecuteCosmeticEvents
    (:False
      (if (Variables|Default|GetbWindowClosed)
        (Collision|SetCollisionEnabled _armoredwindow "QueryAndPhysics")
        (if (ScalableFloat|GetValueasBoolwithContext (Variables|Default|GetHF_AdjustOverlappingPawns))
          (bind _armoredwindow_2 (Variables|Default|GetArmoredWindow)
            (:then
              (bind _outactors (Collision|ComponentOverlapActors _armoredwindow_2 (Transformation|GetWorldTransform _armoredwindow_2) (Utilities|Array|MakeArray "ObjectTypeQuery3") "/Script/FortniteGame.FortPlayerPawnAthena"))
              (for _array_element _outactors
                (bind _asfort_player_pawn (Utilities|Casting|CastToFortPlayerPawn _array_element))
                (bind _charactermovement (Variables|/Script/Engine.Character|GetCharacterMovement _asfort_player_pawn)
                  (:then
                    (Components|Movement|StopMovementImmediately _charactermovement)
                    (Transformation|SetActorLocation _asfort_player_pawn (Utilities|Operators|Add (Transformation|GetActorLocation _asfort_player_pawn) (Math|Vector|MakeVector 0.0 0.0 (ScalableFloat|GetValueatLevel (Variables|Default|GetHF_PawnTeleportOffset)))) false true)
                    (Pawn|Components|CharacterMovement|SetMovementMode _charactermovement "MOVE_Falling" 0))
                  (:else))))
            (:else)))
        (else
          (Collision|SetCollisionEnabled _armoredwindow "NoCollision"))))))

(fn ConstructionScript
  (|CallParentFunction))
